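# Niagara boost effect for the Darkcar vehicle cosmetic, with a sprite Flames emitter and a RIBBON trail emitter.
# Both spawn at User.SingleBoostPos plus their own DTCarOffset parameter.
# Flames plays SubUV sprites that grow and fade, with its paint override disabled.
# RIBBON uses quality-scaled spawning, painted color variants and rearward velocity, inherits source movement with falloff, and draws through four ribbon renderers with brightened, fading color.
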NiagaraSystem /VehicleCosmetics/Boosts/Boost_Darkcar/FX/NS_Boost_Darkcar
dependencies {
 usedDataInterfaces [
  &NiagaraDataInterfaceCurve
  &NiagaraDataInterfaceArrayColor
  &NiagaraDataInterfaceSpriteRendererInfo
  &NiagaraDataInterfaceSkeletalMesh
 ]
 usedDynamicInputs [&Multiply_Vector2D_ByFloat, &RandomRangeVector2D]
 usedModules [
  &self:SystemSpawnScript.SystemScriptSource.SystemScriptGraph.NiagaraNodeAssignment_0.SetVariables_#1
  &SystemState
 ]
}
emitters {
 Flames {
  emitterUpdateScript [
   &EmitterState()
   &SpawnRate(SpawnRate=17, Use Spawn Probability=false)
  ]
  particleSpawnScript [
   &VC_PaintedVariantsOverride(
    Black=&SelectLinearColorFromArray(
     Array Sampling Mode=::NewEnumerator0
     Direct Array Index=?
     Color Selection Array=DataInterface() {propertyValues {ColorData [<rgba 0.0156 0.0156 0.0156 0.5>, <rgba 0.0208 0.0208 0.0208 0.4>]}}
     Random Seed=0
    )
    BurntSienna=&SelectLinearColorFromArray(
     Array Sampling Mode=::NewEnumerator0
     Direct Array Index=?
     Color Selection Array=DataInterface() {propertyValues {ColorData [<rgba 0.3177 0.0688 0 1>, <rgba 0.1198 0.0322 0.0087 1>]}}
     Random Seed=0
    )
    Cobalt=&SelectLinearColorFromArray(
     Array Sampling Mode=::NewEnumerator0
     Direct Array Index=?
     Color Selection Array=DataInterface() {propertyValues {ColorData [<rgba 0 0.0463 2.5 1>, <rgba 0 0.0095 0.5117 1>]}}
     Random Seed=0
    )
    Crimson=&SelectLinearColorFromArray(
     Array Sampling Mode=::NewEnumerator0
     Direct Array Index=?
     Color Selection Array=DataInterface() {propertyValues {ColorData [<rgba 2.25 0 0.061 1>, <rgba 0.3438 0 0.0093 1>]}}
     Random Seed=0
    )
    ForestGreen=&SelectLinearColorFromArray(
     Array Sampling Mode=::NewEnumerator0
     Direct Array Index=?
     Color Selection Array=DataInterface() {propertyValues {ColorData [<rgba 0 0.944 0.0477 1>, <rgba 0 0.3698 0.0187 1.2>]}}
     Random Seed=0
    )
    Grey=&SelectLinearColorFromArray(
     Array Sampling Mode=::NewEnumerator0
     Direct Array Index=?
     Color Selection Array=DataInterface() {propertyValues {ColorData [<rgba 0.2552 0.2552 0.2552 1>, <rgba 0.05 0.05 0.05 1>]}}
     Random Seed=0
    )
    Lime=&SelectLinearColorFromArray(
     Array Sampling Mode=::NewEnumerator0
     Direct Array Index=?
     Color Selection Array=DataInterface() {propertyValues {ColorData [<rgba 0.5 1 0 1>, <rgba 0.2422 0.4844 0 1>]}}
     Random Seed=0
    )
    Orange=&SelectLinearColorFromArray(
     Array Sampling Mode=::NewEnumerator0
     Direct Array Index=?
     Color Selection Array=DataInterface() {propertyValues {ColorData [<rgba 2 0.5915 0.0021 1>, <rgba 0.1198 0.0312 0 1>]}}
     Random Seed=0
    )
    Pink=&SelectLinearColorFromArray(
     Array Sampling Mode=::NewEnumerator0
     Direct Array Index=?
     Color Selection Array=DataInterface() {propertyValues {ColorData [<rgba 2 0.0104 1.2765 1>, <rgba 0.3542 0 0.2254 1>]}}
     Random Seed=0
    )
    Purple=&SelectLinearColorFromArray(
     Array Sampling Mode=::NewEnumerator0
     Direct Array Index=?
     Color Selection Array=DataInterface() {propertyValues {ColorData [<rgba 0.145 0 0.8698 1>, <rgba 0.0521 0 0.3125 1>]}}
     Random Seed=0
    )
    Saffron=&SelectLinearColorFromArray(
     Array Sampling Mode=::NewEnumerator0
     Direct Array Index=?
     Color Selection Array=DataInterface() {propertyValues {ColorData [<rgba 2 1.4416 0 1>, <rgba 0.4167 0.3074 0 2>]}}
     Random Seed=0
    )
    SkyBlue=&SelectLinearColorFromArray(
     Array Sampling Mode=::NewEnumerator0
     Direct Array Index=?
     Color Selection Array=DataInterface() {propertyValues {ColorData [<rgba 0 0.1871 0.2812 1>, <rgba 0 1.3308 2 1>]}}
     Random Seed=0
    )
    TitaniumWhite=&SelectLinearColorFromArray(
     Array Sampling Mode=::NewEnumerator0
     Direct Array Index=?
     Color Selection Array=DataInterface() {propertyValues {ColorData [<rgba 0.4688 0.4688 0.4688 1>, <rgba 2 2 2 1>]}}
     Random Seed=0
    )
   ) {enabled=false}
   &InitializeParticle(
    Color=<rgba 0.6615 0.2099 0.1068 1>
    Lifetime=&RandomRangeFloat(Minimum=0.8, Maximum=0.4)
    Position Offset=&AddVectorToPosition(
     Position=<0 0 0>
     Vector=&AddVectorToPosition(Position=$User.DTCarOffset, Vector=Linked() {linkedVariable {name=User.SingleBoostPos, type=NiagaraPosition}})
    )
    Position Offset Coordinate Space=::Local
    Sprite Rotation Mode=::NewEnumerator2
    Sprite Size Mode=::NewEnumerator1
    Uniform Sprite Size Max=150
    Uniform Sprite Size Min=200
    UsePositionOffset=true
    Write Material Random=true
   )
   &AddVelocity(Velocity=<-150 0 0>) {enabled=false}
   &ShapeLocation(MaskSpawnGroup=false, Override Seed=false, Sphere Radius=35, UseOwnerScale=true) {enabled=false}
  ]
  particleUpdateScript [
   &ParticleState()
   &ScaleSpriteSize(Uniform Curve Sprite Scale=DataInterface() {propertyValues {Curve {keys [{arriveTangent=0.005, interpMode=RCIM_Cubic, leaveTangent=0.005, tangentMode=RCTM_User, time=0, value=0}, <at 1 1 RCIM_Cubic>]}}})
   &SolveForcesAndVelocity(Clamp Velocity=false, Limit Acceleration=false)
   &SubUVAnimation(
    "Sprite Renderer "=DataInterface() {propertyValues {SpriteRenderer=&self:Flames_0.NiagaraSpriteRendererProperties_0}}
    UseEndFrame=false
    UseStartFrame=false
   )
   &ScaleColor(
    Scale Alpha=&FloatFromCurve(FloatCurve=DataInterface() {propertyValues {Curve {keys [<at 0.6 1 RCIM_Cubic>,<at 1 0 RCIM_Cubic>]}}})
    Scale RGB=&VectorFromFloat(Value=&FloatFromCurve(FloatCurve=DataInterface() {propertyValues {Curve {keys [<at 0 1.5>,<at 1 1>]}}}))
    ScaleA=true
    ScaleRGB=true
   )
  ]
  rendererClasses [&NiagaraSpriteRendererProperties]
  renderers [{rendererClass=&NiagaraSpriteRendererProperties}]
 }
 RIBBON {
  emitterUpdateScript [
   &EmitterState()
   &VC_FilteredSocketCount()
   &VC_RateQualityHandler(High Spawn Rate=50, Low Spawn Rate=30, Medium Spawn Rate=45, Use Custom Rate?=true)
   &VC_SpawnRibbon_v1(SpawnRate=$Emitter.Spawn Rate)
  ]
  particleSpawnScript [
   &VC_PaintedVariantsOverride(
    Black=<rgba 0 0 0 2>
    BurntSienna=<rgba 0.4522 0.1948 0.1402 1.2>
    Cobalt=<rgba 0 0.125 1.25 1>
    Crimson=<rgba 0.5729 0.0205 0.0014 1>
    ForestGreen=<rgba 0.0369 0.5365 0.0012 1.5>
    Grey=<rgba 0.4427 0.4427 0.4427 0.8>
    Lime=<rgba 0.825 1.25 0.0007 1>
    Orange=<rgba 1.3 0.2046 0 1>
    Pink=<rgba 1 0.0469 0.6534 1>
    Purple=<rgba 0.1734 0 1.7 1.5>
    Saffron=<rgba 0.8667 0.6644 0 1.5>
    SkyBlue=<rgba 0 0.7985 1.2 1>
    TitaniumWhite=<rgba 1 1 1 1>
   )
   &VC_InitializeRibbon_v1(
    Color=<rgba 1 0.4 0.2 1>
    Custom Offset=true
    Lifetime=1
    Local Vector Offset=&AddVectorToPosition(
     Position=<30 0 0>
     Vector=&AddVectorToPosition(Position=$User.DTCarOffset001, Vector=Linked() {linkedVariable {name=User.SingleBoostPos, type=NiagaraPosition}})
    )
    Offset Amount=1
    Ribbon Width=60
   )
   &AddVelocity(Velocity=<-1400 0 0>) {module=AddVelocity001}
  ]
  particleUpdateScript [
   &ParticleState()
   &InheritSourceMovement(
    Applied Position Scale=&VectorFromFloat(Value=&FloatFromCurve(FloatCurve=DataInterface() {propertyValues {Curve {keys [
          {interpMode=RCIM_Cubic, tangentMode=RCTM_Break, time=0, value=1}
          {arriveTangent=-2.9025, interpMode=RCIM_Cubic, tangentMode=RCTM_Break, time=1, value=0}
         ]}}}))
    ApplyForce=false
    ApplyPosition=true
    ApplyPositionFalloff=false
   )
   &SolveForcesAndVelocity(Clamp Velocity=false, Limit Acceleration=false)
   &VC_UpdateRibbon_v1()
   &ScaleColor(
    Scale Alpha=&FloatFromCurve(FloatCurve=DataInterface() {propertyValues {Curve {
        defaultValue=1
        keys [{interpMode=RCIM_Cubic, tangentMode=RCTM_User, time=0, value=1}, {interpMode=RCIM_Cubic, tangentMode=RCTM_User, time=1, value=0}]
       }}})
    Scale RGB=&VectorFromFloat(Value=&FloatFromCurve(FloatCurve=DataInterface() {propertyValues {Curve {defaultValue=0, keys [{arriveTangent=-1, interpMode=RCIM_Cubic, time=0, value=20}, <at 0.5 5 RCIM_Cubic>]}}}))
    ScaleA=true
    ScaleRGB=true
   )
  ]
  rendererClasses [&NiagaraRibbonRendererProperties]
  renderers [
   {rendererClass=&NiagaraRibbonRendererProperties}
   {rendererClass=&NiagaraRibbonRendererProperties, rendererIndex=1}
   {rendererClass=&NiagaraRibbonRendererProperties, rendererIndex=2}
   {rendererClass=&NiagaraRibbonRendererProperties, rendererIndex=3}
  ]
 }
}
summary {
 systemName=NS_Boost_Darkcar
 emitters [
  {emitterName=RIBBON, rendererClasses [&NiagaraRibbonRendererProperties]}
  {emitterName=Flames, rendererClasses [&NiagaraSpriteRendererProperties]}
 ]
 userVariables [
  User.bIsFrontendPreview {type=NiagaraBool}
  User.bIsPainted {type=NiagaraBool}
  User.ColorVariantID {type=NiagaraInt32}
  User.DTCarOffset {type=NiagaraPosition, description="", defaultValue {}}
  User.DTCarOffset001 {type=NiagaraPosition, description="", defaultValue {}}
  User.GameModeID {type=NiagaraInt32}
  User.PaintedColor {type=LinearColor, description="", defaultValue {struct=LinearColor, value=<rgba 1 1 1 1>}}
  User.PaintedScalar {type=NiagaraFloat, defaultValue=1, description=""}
  User.PortCount {type=NiagaraInt32, defaultValue=1}
  User.SingleBoostPos {type=NiagaraPosition, description="", defaultValue {}}
 ]
}
